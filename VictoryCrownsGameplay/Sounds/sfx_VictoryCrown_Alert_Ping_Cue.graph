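# Sound cue for the Victory Crown alert ping in the VictoryCrownsGameplay content, assigned to the Fort_Gadgets_Items sound class.
# A single wave player feeds the ping wave straight to the output, with no randomization or modulation.
# The cue overrides attenuation inline, using distance-based low-pass filtering to muffle the ping as it gets farther away, along with a wide stereo spread.
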SoundCue /VictoryCrownsGameplay/Sounds/sfx_VictoryCrown_Alert_Ping_Cue
root=WavePlayer_0
nodes [{id=WavePlayer_0, properties {soundWaveAssetPtr=/VictoryCrownsGameplay/Sounds/wav/sfx_VictoryCrown_Alert_Ping}}]
properties {
 bOverrideAttenuation=true
 maxDistance=4100
 soundClassObject=&Fort_Gadgets_Items
 attenuationOverrides {
  bAttenuateWithLPF=true
  lPFFrequencyAtMax=1750
  lPFFrequencyAtMin=10000
  lPFRadiusMax=3000
  lPFRadiusMin=1600
  nonSpatializedRadiusStart=20
  stereoSpread=1600
  attenuationShapeExtents {x=500}
 }
}
