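# Sound cue for the looping ambient sound of the Victory Crown, routed to the backbling sound class.
# A random node picks one of three equally weighted looping waves, and a modulator adds slight volume and pitch variation.
# It uses a shared item ambient-loop attenuation asset.
# Overridden concurrency caps it at two instances, with distance-based volume scaling and a short release for stolen voices.
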SoundCue /VictoryCrownsGameplay/Sounds/sfx_VictoryCrown_Amb_lp_Cue
root=Modulator_0
nodes [
 {id=Modulator_0, children [Random_0], properties {pitchMax=1.005, pitchMin=1, volumeMax=1, volumeMin=0.88}}
 {id=Random_0, children [WavePlayer_0,WavePlayer_1,WavePlayer_2], properties {weights [1,1,1]}}
 {id=WavePlayer_0, properties {bLooping=true, soundWaveAssetPtr=/VictoryCrownsGameplay/Sounds/wav/sfx_VictoryCrown_Amb_LP_01}}
 {id=WavePlayer_1, properties {bLooping=true, soundWaveAssetPtr=/VictoryCrownsGameplay/Sounds/wav/sfx_VictoryCrown_Amb_LP_02}}
 {id=WavePlayer_2, properties {bLooping=true, soundWaveAssetPtr=/VictoryCrownsGameplay/Sounds/wav/sfx_VictoryCrown_Amb_LP_03}}
]
properties {
 attenuationSettings=/Game/Sounds/Attenuation/Fort_Item_Ambient_Loop_Mono_Default_Att
 bDebug=true
 bOverrideConcurrency=true
 maxDistance=1250
 soundClassObject=/Game/Sounds/FortSoundClassesAndModes/Fort_Backblings
 attenuationOverrides {falloffDistance=512, nonSpatializedRadiusStart=20, stereoSpread=420}
 concurrencyOverrides {maxCount=2, voiceStealReleaseTime=0.3, volumeScale=0.4, volumeScaleAttackTime=0.1, volumeScaleMode=Distance}
}
