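# Sound cue for the Victory Crown drop sound in the VictoryCrownsGameplay plugin.
# A Random node picks one of three drop wave variants with equal weights, then a Modulator at its default settings feeds the output.
# It is spatialized through the shared Fort weapon-drop attenuation and routed to the Athena player foley sound class.
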SoundCue /VictoryCrownsGameplay/Sounds/sfx_VictoryCrown_Drop_Cue
root=Modulator_0
nodes [
 {id=Modulator_0, children [Random_1]}
 {id=Random_1, children [WavePlayer_3,WavePlayer_4,WavePlayer_5], properties {weights [1,1,1]}}
 {id=WavePlayer_3, properties {soundWaveAssetPtr=/VictoryCrownsGameplay/Sounds/wav/sfx_VictoryCrown_Drop_01}}
 {id=WavePlayer_4, properties {soundWaveAssetPtr=/VictoryCrownsGameplay/Sounds/wav/sfx_VictoryCrown_Drop_02}}
 {id=WavePlayer_5, properties {soundWaveAssetPtr=/VictoryCrownsGameplay/Sounds/wav/sfx_VictoryCrown_Drop_03}}
]
properties {
 attenuationSettings=&Fort_Weapon_Drop_Att
 maxDistance=3000
 soundClassObject=&Fort_Player_Foley_Athena
 volumeMultiplier=1
}
