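# Sound cue for the Victory Crown flair effect, playing a single accolade wave.
# A modulator feeds the wave through a very slight random upward pitch variation.
# It routes to the Fort_UI sound class yet overrides attenuation for distance falloff and stereo spread.
# A concurrency override caps simultaneous instances per owner and sets a short retrigger time.
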SoundCue /VictoryCrownsGameplay/Sounds/sfx_VictoryCrown_Flair_Cue
root=Modulator_1
nodes [
 {id=Modulator_1, children [WavePlayer_0], properties {pitchMax=1.005, pitchMin=1}}
 {id=WavePlayer_0, properties {soundWaveAssetPtr=/VictoryCrownsGameplay/Sounds/wav/sfx_VictoryCrown_Accolade_01}}
]
properties {
 bOverrideAttenuation=true
 bOverrideConcurrency=true
 maxDistance=7500
 soundClassObject=&Fort_UI
 volumeMultiplier=0.9
 attenuationOverrides {falloffDistance=7000, nonSpatializedRadiusStart=20, stereoSpread=900, attenuationShapeExtents {x=500}}
 concurrencyOverrides {bLimitToOwner=true, maxCount=2, retriggerTime=0.3}
}
