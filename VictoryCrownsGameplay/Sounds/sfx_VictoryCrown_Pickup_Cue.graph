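# Sound cue that plays the pickup sound for a Victory Crown.
# A Random node picks one of three equally weighted wave variants, which feeds a Modulator at default settings for slight pitch and volume variation.
# It borrows the weapon pickup attenuation and loot collect sound class, and plays at reduced volume.
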SoundCue /VictoryCrownsGameplay/Sounds/sfx_VictoryCrown_Pickup_Cue
root=Modulator_0
nodes [
 {id=Modulator_0, children [Random_1]}
 {id=Random_1, children [WavePlayer_0,WavePlayer_1,WavePlayer_2], properties {weights [1,1,1]}}
 {id=WavePlayer_0, properties {soundWaveAssetPtr=/VictoryCrownsGameplay/Sounds/wav/sfx_VictoryCrown_Pickup_01}}
 {id=WavePlayer_1, properties {soundWaveAssetPtr=/VictoryCrownsGameplay/Sounds/wav/sfx_VictoryCrown_Pickup_02}}
 {id=WavePlayer_2, properties {soundWaveAssetPtr=/VictoryCrownsGameplay/Sounds/wav/sfx_VictoryCrown_Pickup_03}}
]
properties {
 attenuationSettings=&Fort_Weapon_Pickup_Att
 maxDistance=3500
 soundClassObject=&Fort_Loot_Collect
 volumeMultiplier=0.398
}
