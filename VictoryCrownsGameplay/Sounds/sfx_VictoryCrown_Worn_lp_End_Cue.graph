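# SoundCue that plays the end sound for the worn Victory Crown loop, using the third-person (3P) wave.
# A single Wave Player feeds a Modulator that adds slight random pitch and volume variation.
# The cue boosts its volume, routes through the Fort_UI sound class, and overrides attenuation with its own inner radius and falloff distance.
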SoundCue /VictoryCrownsGameplay/Sounds/sfx_VictoryCrown_Worn_lp_End_Cue
root=Modulator_1
nodes [
 {id=Modulator_1, children [WavePlayer_0], properties {pitchMax=1.005, pitchMin=1, volumeMax=1.2, volumeMin=1}}
 {id=WavePlayer_0, properties {soundWaveAssetPtr=/VictoryCrownsGameplay/Sounds/wav/sfx_VictoryCrown_Worn_3P_lp_End}}
]
properties {
 bOverrideAttenuation=true
 maxDistance=2650
 soundClassObject=&Fort_UI
 volumeMultiplier=1.3
 attenuationOverrides {falloffDistance=2000, attenuationShapeExtents {x=650}}
}
